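# Niagara system for the Shimpilon glider's idle effect, with emitters including mesh bursts at eyes, limbs and tail plus drifting sprite smoke.
# Mesh particles follow skeletal mesh samples and drive dynamic material parameters for animated, fading shapes.
# User.ForwardDot blends loop timing and spawn rates, also shaping several lifetimes and scales.
# User.VariantSwitch selects colours and brightness, while front-end modifiers swap timing, offsets and intensity for preview versus gameplay.
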NiagaraSystem /BRCosmetics/Gadgets/Assets/VinderTech_GliderChute/Glider_ShimPilon/FX/NS_Glider_Shimpilon_Idle
dependencies {
 usedDataInterfaces [
  &NiagaraDataInterfaceArrayColor
  &NiagaraDataInterfaceColorCurve
  &NiagaraDataInterfaceCurve
  &NiagaraDataInterfaceArrayFloat
  &NiagaraDataInterfaceArrayFloat3
  &NiagaraDataInterfacePlatformSet
  &NiagaraDataInterfaceSkeletalMesh
 ]
 usedDynamicInputs [
  &MakeVector4
  &MakeCustomFloatFromBool
  &SetBoolByFloatComparison
 ]
 usedModules [
  &self:SystemSpawnScript.SystemScriptSource.SystemScriptGraph.NiagaraNodeAssignment_0.SetVariables_#1
  &self:SystemSpawnScript.SystemScriptSource.SystemScriptGraph.NiagaraNodeAssignment_1.SetVariables_#2
  &self:SystemSpawnScript.SystemScriptSource.SystemScriptGraph.NiagaraNodeAssignment_2.SetVariables_#3
  &SystemState
  &InterpolateOverTime
 ]
}
emitters {
 Eyes {
  emitterSpawnScript [
   /SetVariables(Emitter.DepthFade=20) {isSetParameters=true, module=SetVariables_#4}
   /SetVariables(Emitter.Width=40) {isSetParameters=true}
  ]
  emitterUpdateScript [
   &EmitterState(
    Delay First Loop Only=true
    Inactive Response=::NewEnumerator1
    Life Cycle Mode=::NewEnumerator1
    Loop Delay=&FN_FrontEndFloatModifier(
     Gameplay=&Lerp_Float(
      A=&RandomRangeFloat(Minimum=0.5, Maximum=1)
      B=&RandomRangeFloat(Minimum=0.5, Maximum=0)
      Alpha=&Lerp_Float(A=$System.Test, B=1, Alpha=$User.ForwardDot)
     )
     Front End=&Lerp_Float(A=$System.Front End Delay, B=0, Alpha=$System.Test)
    )
    Loop Duration=&FN_FrontEndFloatModifier(
     Gameplay=&Lerp_Float(
      A=&RandomRangeFloat(Minimum=0.5, Maximum=1)
      B=&RandomRangeFloat(Minimum=0.5, Maximum=0.2)
      Alpha=&Lerp_Float(A=$System.Test, B=1, Alpha=$User.ForwardDot)
     )
     Front End=&RandomRangeFloat(Minimum=0.5, Maximum=0.2)
    )
    Recalculate Duration Each Loop=true
    UseLoopDelay=true
   )
   &SpawnRate(
    SpawnRate=&FN_FrontEndFloatModifier(
     Gameplay=&Lerp_Float(A=1, B=10, Alpha=&Lerp_Float(A=$System.Test, B=1, Alpha=$User.ForwardDot))
     Front End=8
    )
    Use Spawn Probability=false
   )
  ]
  particleSpawnScript [
   &PartitionParticles()
   &InitializeParticle(
    Color Mode=::NewEnumerator0
    Lifetime=&Lerp_Float(
     A=&RandomRangeFloat(Minimum=0.3, Maximum=1)
     B=&RandomRangeFloat(Minimum=0.3, Maximum=0.6)
     Alpha=$User.ForwardDot
    )
    Mesh Scale Max=<1.5 1.5 1.5>
    Mesh Scale Mode=::NewEnumerator2
    UsePositionOffset=false
    Write Material Random=true
   )
   &CameraOffset(Camera Offset Amount=10)
   &InitialMeshOrientation(
    Mesh Orientation Mode=::NewEnumerator2
    Rotation=&SelectVectorFromArray(
     Direct Array Index=$Particles.PartitionParticles.Partition
     Vector Selection Array=DataInterface() {propertyValues {FloatData [<0 -0.2 0>, <0 0.2 0>]}}
    )
    Use Rotation Vector=true
   )
   &InitialMeshOrientation(
    Mesh Orientation Mode=::NewEnumerator2
    Rotation=&RandomRangeVector(
     Minimum=&SelectVectorFromArray(
      Direct Array Index=$Particles.PartitionParticles.Partition
      Vector Selection Array=DataInterface() {propertyValues {FloatData [<0 0 0>, <0 0 0>]}}
     )
     Maximum=&FN_FrontEndVector3Modifier(
      Gameplay=&SelectVectorFromArray(
       Direct Array Index=$Particles.PartitionParticles.Partition
       Vector Selection Array=DataInterface() {propertyValues {FloatData [<0 -0.15 0>, <0 0.15 0>]}}
      )
      Front End=&SelectVectorFromArray(
       Direct Array Index=$Particles.PartitionParticles.Partition
       Vector Selection Array=DataInterface() {propertyValues {FloatData [<0 0 0>, <0 0.15 0>]}}
      )
     )
    )
    Use Rotation Vector=true
   ) {module=InitialMeshOrientation001}
   &InitialMeshOrientation(
    Mesh Orientation Mode=::NewEnumerator2
    Rotation=&FN_FrontEndVector3Modifier(Gameplay=<0.1 0 0>, Front End=<0.2 0 0>)
    Use Rotation Vector=true
   ) {module=InitialMeshOrientation002}
   &InitialMeshOrientation(
    Mesh Orientation Mode=::NewEnumerator2
    Rotation=&FN_FrontEndVector3Modifier(
     Gameplay=<0 0 0>
     Front End=&SelectVectorFromArray(
      Direct Array Index=$Particles.PartitionParticles.Partition
      Vector Selection Array=DataInterface() {propertyValues {FloatData [<0 -0.05 0>, <0 0.05 0>]}}
     )
    )
    Use Rotation Vector=true
   ) {module=InitialMeshOrientation003}
  ]
  particleUpdateScript [
   &ParticleState()
   &SkeletalMeshLocation(
    Bone Sampling Mode=::NewEnumerator4
    MaskSpawnGroup=false
    Orientation Sampling=::NewEnumerator1
    Sampled Position Offset=&FN_FrontEndVector3Modifier(
     Gameplay=&SelectVectorFromArray(
      Direct Array Index=$Particles.PartitionParticles.Partition
      Vector Selection Array=DataInterface() {propertyValues {FloatData [<20 -20 -15>, <20 -20 15>]}}
     )
     Front End=&SelectVectorFromArray(
      Direct Array Index=$Particles.PartitionParticles.Partition
      Vector Selection Array=DataInterface() {propertyValues {FloatData [<15 -12 -5>, <15 -15 5>]}}
     )
    )
    Skeletal Mesh=$System.Skeletal Mesh
   )
   &self:VariantColor(
    DefaultColor:LinearColor=&ColorFromCurve(ColorCurve=DataInterface() {propertyValues {
       AlphaCurve {keys [<at 0.4428 1>,<at 1 0 RCIM_Cubic>]}
       BlueCurve {keys [<at 0.2021 1 RCIM_Cubic>,<at 0.4091 1>]}
       GreenCurve {keys [<at 0.2021 0 RCIM_Cubic>,<at 0.4091 0>]}
       RedCurve {keys [<at 0.2021 0.5 RCIM_Cubic>,<at 0.4091 0.1667>]}
      }})
    PipColor:LinearColor=&ColorFromCurve(ColorCurve=DataInterface() {propertyValues {
       AlphaCurve {keys [<at 0 1 RCIM_Cubic>,<at 1 0 RCIM_Cubic>]}
       BlueCurve {keys [<at 0.1471 1 RCIM_Cubic>,<at 0.4146 1>]}
       GreenCurve {keys [<at 0.1471 0.6667 RCIM_Cubic>,<at 0.4146 0.1667>]}
       RedCurve {keys [<at 0.1471 0 RCIM_Cubic>,<at 0.4146 0>]}
      }})
    SeedColor:LinearColor=<rgba 1 0 0.3333 1>
    StoneColor:LinearColor=<rgba 0 0 0 0>
    VariantIndex:NiagaraInt32=&MakeIntFromFloat(Float=$User.VariantSwitch)
   ) {module=VariantColor}
   &ScaleMeshSize(Scale Factor=&MakeVector(
     X=1
     Y=&FloatFromCurve(
      FloatCurve=DataInterface() {propertyValues {Curve {keys [
          {arriveTangent=4.4858, interpMode=RCIM_Cubic, leaveTangent=4.4858, tangentMode=RCTM_User, time=0, value=0.2}
          {arriveTangent=0.3309, interpMode=RCIM_Cubic, leaveTangent=0.3309, tangentMode=RCTM_User, time=0.0832, value=0.8453}
          {arriveTangent=-0.0088, interpMode=RCIM_Cubic, leaveTangent=-0.0088, tangentMode=RCTM_User, time=1, value=1}
         ]}}}
      Scale Curve=&Lerp_Float(
       A=&RandomRangeFloat(Minimum=0.8, Maximum=0.6)
       B=&RandomRangeFloat(Minimum=1.5, Maximum=1.2)
       Alpha=$User.ForwardDot
      )
     )
     Z=1
    )) {module=ScaleMeshSize002}
   &ScaleMeshSize(Scale Factor=&MakeCustomVectorFromBool(Bool=&RandomBool(), True Vector=<-1 1 1>, False Vector=<1 1 1>))
   &DynamicMaterialParameters(
    Index 0 Param 1=&FloatFromCurve(
     FloatCurve=DataInterface() {propertyValues {Curve {keys [
         {interpMode=RCIM_Cubic, tangentMode=RCTM_Break, time=0, value=0.2}
         {arriveTangent=2.7275, interpMode=RCIM_Cubic, tangentMode=RCTM_Break, time=0.8, value=1}
        ]}}}
     Scale Curve=&RandomRangeFloat(Minimum=1, Maximum=1.2)
    )
    Index 0 Param 2=0.7
    Index 0 Param 3=&FloatFromCurve(
     FloatCurve=DataInterface() {propertyValues {Curve {keys [
         {arriveTangent=1.5447, interpMode=RCIM_Cubic, leaveTangent=1.5447, tangentMode=RCTM_User, time=0, value=0}
         {arriveTangent=0.0259, interpMode=RCIM_Cubic, leaveTangent=0.0259, tangentMode=RCTM_User, time=1, value=3}
        ]}}}
     Scale Curve=&RandomRangeFloat(Minimum=1, Maximum=2)
    )
    Index 1 Param 1=&Multiply_Float(
     A=&RandomRangeFloat(Minimum=1, Maximum=2)
     B=&Lerp_Float(A=0.65, B=0.8, Alpha=$User.ForwardDot)
    )
    Index 1 Param 2=&RandomRangeFloat(Minimum=0, Maximum=1)
    Index 1 Param 3=&Lerp_Float(A=0.4, B=0.6, Alpha=$User.ForwardDot)
    Index 1 Param 4=1
    Index 2 Param 1=&FloatFromCurve(FloatCurve=DataInterface() {propertyValues {Curve {keys [{tangentMode=RCTM_User, time=0, value=0.2}, {tangentMode=RCTM_User, time=1, value=1}]}}})
    Index 2 Param 4=5
    Module.Index0Float0Write=true
    Module.Index0Float1Write=true
    Module.Index0Float2Write=true
    Module.Index0Float3Write=true
    Module.Index1Float0Write=true
    Module.Index1Float1Write=true
    Module.Index1Float2Write=true
    Module.Index1Float3Write=true
    Module.Index2Float0Write=true
    Module.Index2Float1Write=true
    Module.Index2Float2Write=true
    Module.Index2Float3Write=true
    Write Parameter Index 1=::NewEnumerator1
    Write Parameter Index 2=::NewEnumerator1
   )
   &FN_Mesh_Attributes(
    Color=&MakeLinearColorFromVectorAndFloat(
     "Vector (RGB)"=&Multiply_VectorByFloat(Vector=&MakeVectorFromLinearColorRGB(Color=$Particles.Color), Float=8)
     "Float (Alpha)"=&FloatFromCurve(
      FloatCurve=DataInterface() {propertyValues {Curve {keys [
          <at 0 0>
          <at 0.1 0>
          <at 0.2 0.8 RCIM_Cubic>
          {arriveTangent=-2.3099, interpMode=RCIM_Cubic, leaveTangent=-2.3099, tangentMode=RCTM_User, time=1, value=0}
         ]}}}
      Scale Curve=6
     )
    )
    Create Color Attribute=true
    Create Dynamic Parameter Attribute=true
    Index 0 Param 1:NiagaraFloat=&FloatFromCurve(FloatCurve=DataInterface() {propertyValues {Curve {keys [{tangentMode=RCTM_User, time=0, value=0.2}, {tangentMode=RCTM_User, time=1, value=1}]}}})
    Index 0 Param 2=0.7
    Index 0 Param 3:NiagaraFloat=&FloatFromCurve(
     FloatCurve=DataInterface() {propertyValues {Curve {keys [<at 0 0 RCIM_Cubic>, {arriveTangent=2.5076, interpMode=RCIM_Cubic, leaveTangent=2.5076, tangentMode=RCTM_User, time=1, value=3}]}}}
     Scale Curve=&RandomRangeFloat(Minimum=2, Maximum=1)
    )
    Module Output=::NewEnumerator1
    Module.Index0Float0Write=true
    Module.Index0Float1Write=true
    Module.Index0Float2Write=true
    Module.Index0Float3Write=true
   )
   &FN_Mesh_Attributes(
    Color=&MakeLinearColorFromVectorAndFloat(
     "Vector (RGB)"=&Multiply_VectorByFloat(
      Vector=&MakeVectorFromLinearColorRGB(Color=$Particles.Color)
      Float=&FN_FrontEndFloatModifier(
       Gameplay=&SelectFloatFromArray(
        Direct Array Index=&MakeIntFromFloat(Float=$User.VariantSwitch)
        Float Selection Array=DataInterface() {propertyValues {FloatData [2,2,1.5]}}
       )
       Front End=&SelectFloatFromArray(
        Direct Array Index=&MakeIntFromFloat(Float=$User.VariantSwitch)
        Float Selection Array=DataInterface() {propertyValues {FloatData [2,1,1]}}
       )
      )
     )
     "Float (Alpha)"=&FloatFromCurve(
      FloatCurve=DataInterface() {propertyValues {Curve {keys [
          <at 0 1>
          {arriveTangent=1.3985, time=0.2, value=1}
          {arriveTangent=-2, interpMode=RCIM_Cubic, leaveTangent=-2, time=0.35, value=1}
          {arriveTangent=-1.3423, interpMode=RCIM_Cubic, leaveTangent=-1.3423, tangentMode=RCTM_User, time=0.7, value=0}
         ]}}}
      Scale Curve=&FN_FrontEndFloatModifier(
       Gameplay=2.5
       Front End=&SelectFloatFromArray(
        Direct Array Index=&MakeIntFromFloat(Float=$User.VariantSwitch)
        Float Selection Array=DataInterface() {propertyValues {FloatData [2.5,2.5,1.5]}}
       )
      )
     )
    )
    Create Color Attribute=true
    Create Dynamic Parameter Attribute=true
    Index 0 Param 1:NiagaraFloat=&FloatFromCurve(FloatCurve=DataInterface() {propertyValues {Curve {keys [{tangentMode=RCTM_User, time=0, value=0.2}, {tangentMode=RCTM_User, time=1, value=1}]}}})
    Index 0 Param 2=0.7
    Module Output=::NewEnumerator1
    Module.Index0Float0Write=true
    Module.Index0Float1Write=true
    Module.Index0Float2Write=true
    Module.Index0Float3Write=true
   ) {module=FN_Mesh_Attributes001}
   &ScaleColor(
    Color Value To Scale=$Particles.Color
    Scale Alpha=&Multiply_Float(
     A=5
     B=&FloatFromCurve(FloatCurve=DataInterface() {propertyValues {Curve {keys [<at 0 0>,<at 0.1 0>,<at 0.2 0.8>,<at 1 0>]}}})
    )
    Scale RGB=&VectorFromFloat(Value=&RandomRangeFloat(Minimum=0.5, Maximum=1, Evaluation Type=::NewEnumerator1))
    ScaleA=true
    ScaleRGB=true
   )
  ]
  rendererClasses [&NiagaraMeshRendererProperties]
  renderers [
   {rendererClass=&NiagaraMeshRendererProperties, rendererIndex=0}
   {rendererClass=&NiagaraMeshRendererProperties, rendererIndex=1}
   {rendererClass=&NiagaraMeshRendererProperties, rendererIndex=2}
  ]
 }
 Forward_smoke {
  emitterUpdateScript [
   &EmitterState(
    Delay First Loop Only=true
    Inactive Response=::NewEnumerator1
    Life Cycle Mode=::NewEnumerator1
    Loop Delay=&FN_FrontEndFloatModifier(
     Gameplay=&Lerp_Float(
      A=&RandomRangeFloat(Minimum=0.5, Maximum=3)
      B=&RandomRangeFloat(Minimum=0.2, Maximum=0)
      Alpha=&Lerp_Float(A=$System.Test, B=1, Alpha=$User.ForwardDot)
     )
     Front End=&Lerp_Float(A=$System.Front End Delay, B=0, Alpha=$System.Test)
    )
    Loop Duration=&FN_FrontEndFloatModifier(
     Gameplay=&Lerp_Float(
      A=&RandomRangeFloat(Minimum=0.5, Maximum=1)
      B=&RandomRangeFloat(Minimum=0.5, Maximum=0.2)
      Alpha=&Lerp_Float(A=$System.Test, B=1, Alpha=$User.ForwardDot)
     )
     Front End=&RandomRangeFloat(Minimum=0.5, Maximum=0.2)
    )
    Recalculate Duration Each Loop=true
    UseLoopDelay=true
   )
   &SpawnRate(
    SpawnRate=&FN_FrontEndFloatModifier(
     Gameplay=&Lerp_Float(A=0, B=10, Alpha=&Lerp_Float(A=$System.Test, B=1, Alpha=$User.ForwardDot))
     Front End=10
    )
    Use Spawn Probability=false
   )
  ]
  particleSpawnScript [
   &InitializeParticle(
    Color Mode=::NewEnumerator0
    Lifetime Max=1
    Lifetime Min=0.8
    Lifetime Mode=::NewEnumerator1
    Sprite Rotation Mode=::NewEnumerator2
    Sprite Size Mode=::NewEnumerator1
    Sprite UV Mode=::NewEnumerator3
    Uniform Sprite Size Max=200
    Uniform Sprite Size Min=150
    UsePositionOffset=false
    Write Material Random=true
   )
   &ShapeLocation(
    Box Size=<150 250 0>
    MaskSpawnGroup=false
    Offset=&FN_FrontEndVector3Modifier(Gameplay=<0 0 0>, Front End=<-200 -200 0>)
    Offset Mode=::NewEnumerator0
    Override Seed=false
    Shape Primitive=::NewEnumerator2
    Surface Only=false
    UseOwnerScale=true
   )
   &AddVelocity(Velocity=&RandomRangeVector(Minimum=<0 800 0>, Maximum=<0 1600 0>), Velocity Speed Scale=0.8)
  ]
  particleUpdateScript [
   &ParticleState()
   &SpriteRotationRate(Rotation Rate=&RandomRangeFloat(Minimum=-50, Maximum=50))
   &ScaleSpriteSize(
    Uniform Curve Scale=&FN_FrontEndFloatModifier(Gameplay=1, Front End=0.8)
    Uniform Curve Sprite Scale=DataInterface() {propertyValues {Curve {
       defaultValue=1
       keys [{interpMode=RCIM_Cubic, leaveTangent=1.2633, tangentMode=RCTM_Break, time=0, value=0.3}, {tangentMode=RCTM_User, time=1, value=1}]
      }}}
   )
   &Drag(Use Linear Drag=true, Use Rotational Drag=true)
   &SolveForcesAndVelocity(Clamp Velocity=false, Limit Acceleration=false)
   &self:VariantColor(
    DefaultColor:LinearColor=&ColorFromCurve(ColorCurve=DataInterface() {propertyValues {
       AlphaCurve {keys [<at 0.4428 1>,<at 1 0 RCIM_Cubic>]}
       BlueCurve {keys [<at 0.2021 1 RCIM_Cubic>,<at 0.412 1>,<at 0.9952 1>]}
       GreenCurve {keys [<at 0.2021 0.05 RCIM_Cubic>,<at 0.412 0.0052>,<at 0.9952 0>]}
       RedCurve {keys [<at 0.2021 0.525 RCIM_Cubic>,<at 0.412 0.171>,<at 0.9952 0.1667>]}
      }})
    PipColor:LinearColor=&ColorFromCurve(ColorCurve=DataInterface() {propertyValues {
       AlphaCurve {keys [<at 0 1 RCIM_Cubic>,<at 1 0 RCIM_Cubic>]}
       BlueCurve {keys [<at 0 1 RCIM_Cubic>,<at 0.5607 1>]}
       GreenCurve {keys [<at 0 0.6667 RCIM_Cubic>,<at 0.5607 0.1667>]}
       RedCurve {keys [<at 0 0 RCIM_Cubic>,<at 0.5607 0>]}
      }})
    SeedColor:LinearColor=&MakeCustomLinearColorFromBool(Bool=$User.bIsFrontEndPreview, True Color=<rgba 1 0 0.3333 1>, False Color=<rgba 1 0 0.3333 1>)
    StoneColor:LinearColor=<rgba 0 0 0 0>
    VariantIndex:NiagaraInt32=&MakeIntFromFloat(Float=$User.VariantSwitch)
   ) {module=VariantColor}
   &ScaleColor(
    Color Value To Scale=$Particles.Color
    Scale Alpha=&FloatFromCurve(
     FloatCurve=DataInterface() {propertyValues {Curve {
        defaultValue=0
        keys [
         {interpMode=RCIM_Cubic, tangentMode=RCTM_User, time=0, value=0}
         {interpMode=RCIM_Cubic, tangentMode=RCTM_User, time=0.5, value=1}
         {tangentMode=RCTM_User, time=1, value=0}
        ]
       }}}
     Scale Curve=&FN_FrontEndFloatModifier(
      Gameplay=&FN_PickFloatDependOnPlatform(
       TargetPlatformActive=&SelectFloatFromArray(
        Direct Array Index=&MakeIntFromFloat(Float=$User.VariantSwitch)
        Float Selection Array=DataInterface() {propertyValues {FloatData [1,1,0.5]}}
       )
       TargetPlatformNoActive=&SelectFloatFromArray(
        Direct Array Index=&MakeIntFromFloat(Float=$User.VariantSwitch)
        Float Selection Array=DataInterface() {propertyValues {FloatData [0.3,0.3,0.2]}}
       )
      )
      Front End=&SelectFloatFromArray(
       Direct Array Index=&MakeIntFromFloat(Float=$User.VariantSwitch)
       Float Selection Array=DataInterface() {propertyValues {FloatData [0.6,0.6,0.2]}}
      )
     )
    )
    ScaleA=true
    ScaleRGB=true
   )
  ]
  rendererClasses [&NiagaraSpriteRendererProperties]
  renderers [{rendererClass=&NiagaraSpriteRendererProperties}]
 }
 Limbs {
  emitterSpawnScript [
   /SetVariables(Emitter.DepthFade=10) {isSetParameters=true, module=SetVariables_#4}
   /SetVariables(Emitter.Width:NiagaraFloat=&Lerp_Float(A=40, B=50, Alpha=$User.ForwardDot)) {isSetParameters=true, module=SetVariables_#5}
  ]
  emitterUpdateScript [
   &EmitterState(
    Delay First Loop Only=true
    Inactive Response=::NewEnumerator1
    Life Cycle Mode=::NewEnumerator1
    Loop Delay=&FN_FrontEndFloatModifier(
     Gameplay=&Lerp_Float(
      A=&RandomRangeFloat(Minimum=0.5, Maximum=3)
      B=&RandomRangeFloat(Minimum=0.5, Maximum=0)
      Alpha=&Lerp_Float(A=$System.Test, B=1, Alpha=$User.ForwardDot)
     )
     Front End=&Lerp_Float(A=$System.Front End Delay, B=0, Alpha=$System.Test)
    )
    Loop Duration=&FN_FrontEndFloatModifier(
     Gameplay=&Lerp_Float(
      A=&RandomRangeFloat(Minimum=0.5, Maximum=1)
      B=&RandomRangeFloat(Minimum=0.5, Maximum=0.2)
      Alpha=&Lerp_Float(A=$System.Test, B=1, Alpha=$User.ForwardDot)
     )
     Front End=&RandomRangeFloat(Minimum=0.5, Maximum=0.2)
    )
    Recalculate Duration Each Loop=true
    UseLoopDelay=true
   )
   &SpawnRate(
    SpawnRate=&FN_FrontEndFloatModifier(
     Gameplay=&Lerp_Float(A=3, B=20, Alpha=&Lerp_Float(A=$System.Test, B=1, Alpha=$User.ForwardDot))
     Front End=15
    )
    Use Spawn Probability=false
   )
  ]
  particleSpawnScript [
   &PartitionParticles()
   &InitializeParticle(
    Color=&SelectLinearColorFromArray(
     Array Sampling Mode=::NewEnumerator1
     Direct Array Index=&MakeIntFromFloat(Float=$User.VariantSwitch)
     Color Selection Array=DataInterface() {propertyValues {ColorData [<rgba 0.0833 0 0.5 1>, <rgba 0 0 0 0>]}}
     Random Seed=?
    )
    Lifetime=&Lerp_Float(
     A=&RandomRangeFloat(Minimum=0.3, Maximum=0.7)
     B=&RandomRangeFloat(Minimum=0.2, Maximum=0.5)
     Alpha=$User.ForwardDot
    )
    Mesh Scale Max=<1.5 1 1>
    Mesh Scale Min=<1 0.5 0.5>
    Mesh Scale Mode=::NewEnumerator2
    UsePositionOffset=false
    Write Material Random=true
   )
   &CameraOffset(Camera Offset Amount=10)
   &ShapeLocation(
    Box Size=<5 100 50>
    MaskSpawnGroup=false
    Override Seed=false
    Shape Primitive=::NewEnumerator2
    Surface Only=true
    Surface Only Band Thickness=1
    UseOwnerScale=true
   ) {module=ShapeLocation001}
   &InitialMeshOrientation(
    Mesh Orientation Mode=::NewEnumerator2
    Rotation=&RandomRangeVector(Minimum=<0 0 0>, Maximum=<1 0 0>)
    Use Rotation Vector=true
   ) {module=InitialMeshOrientation001}
   &InitialMeshOrientation(Mesh Orientation Mode=::NewEnumerator2, Rotation=<0 0 0.25>, Use Rotation Vector=true)
   &InitialMeshOrientation(
    Mesh Orientation Mode=::NewEnumerator2
    Rotation=&FN_FrontEndVector3Modifier(
     Gameplay=&SelectVectorFromArray(
      Direct Array Index=$Particles.PartitionParticles.Partition
      Vector Selection Array=DataInterface() {propertyValues {FloatData [<0 0 0.05>, <0 0 -0.05>]}}
     )
     Front End=&SelectVectorFromArray(
      Direct Array Index=$Particles.PartitionParticles.Partition
      Vector Selection Array=DataInterface() {propertyValues {FloatData [<0 0 0.1>, <0 0 -0.03>]}}
     )
    )
    Use Rotation Vector=true
   ) {module=InitialMeshOrientation002}
   &InitialMeshOrientation(
    Mesh Orientation Mode=::NewEnumerator2
    Rotation=&RandomRangeVector(Minimum=<0 0 0>, Maximum=<-0.03 0 0>)
    Use Rotation Vector=true
   ) {module=InitialMeshOrientation003}
  ]
  particleUpdateScript [
   &ParticleState()
   &SkeletalMeshLocation(
    Bone / Socket Index=6
    Bone Sampling Mode=::NewEnumerator4
    MaskSpawnGroup=false
    Orientation Sampling=::NewEnumerator1
    Sampled Position Offset=&AddVectorToPosition(
     Position=Linked() {linkedVariable {name=Particles.Initial.ShapeLocation001.ShapeVector, type=Vector3f}}
     Vector=&FN_FrontEndVector3Modifier(
      Gameplay=&SelectVectorFromArray(
       Direct Array Index=$Particles.PartitionParticles.Partition
       Vector Selection Array=DataInterface() {propertyValues {FloatData [<-40 0 -40>, <40 0 -40>]}}
      )
      Front End=&SelectVectorFromArray(
       Direct Array Index=$Particles.PartitionParticles.Partition
       Vector Selection Array=DataInterface() {propertyValues {FloatData [<-30 0 -30>, <30 0 -30>]}}
      )
     )
    )
    Skeletal Mesh=$System.Skeletal Mesh
   )
   &self:VariantColor(
    DefaultColor:LinearColor=&ColorFromCurve(ColorCurve=DataInterface() {propertyValues {
       AlphaCurve {keys [<at 0.4428 1>,<at 1 0 RCIM_Cubic>]}
       BlueCurve {keys [<at 0.2021 1 RCIM_Cubic>,<at 0.4091 1>]}
       GreenCurve {keys [<at 0.2021 0 RCIM_Cubic>,<at 0.4091 0>]}
       RedCurve {keys [<at 0.2021 0.5 RCIM_Cubic>,<at 0.4091 0.1667>]}
      }})
    PipColor:LinearColor=&ColorFromCurve(ColorCurve=DataInterface() {propertyValues {
       AlphaCurve {keys [<at 0 1 RCIM_Cubic>,<at 1 0 RCIM_Cubic>]}
       BlueCurve {keys [<at 0.1471 1 RCIM_Cubic>,<at 0.4146 1>]}
       GreenCurve {keys [<at 0.1471 0.6667 RCIM_Cubic>,<at 0.4146 0.1667>]}
       RedCurve {keys [<at 0.1471 0 RCIM_Cubic>,<at 0.4146 0>]}
      }})
    SeedColor:LinearColor=<rgba 1 0 0.3333 1>
    StoneColor:LinearColor=<rgba 0 0 0 0>
    VariantIndex:NiagaraInt32=&MakeIntFromFloat(Float=$User.VariantSwitch)
   ) {module=VariantColor}
   &ScaleMeshSize(Scale Factor=&MakeVector(
     X=&FloatFromCurve(
      FloatCurve=DataInterface() {propertyValues {Curve {keys [
          {arriveTangent=0.3309, interpMode=RCIM_Cubic, leaveTangent=0.3309, tangentMode=RCTM_User, time=0.2, value=0.8453}
          {arriveTangent=-0.0088, interpMode=RCIM_Cubic, leaveTangent=-0.0088, tangentMode=RCTM_User, time=1, value=1}
         ]}}}
      Scale Curve=&Lerp_Float(A=0.7, B=&RandomRangeFloat(Minimum=0.6, Maximum=0.8), Alpha=$User.ForwardDot)
     )
     Y=1
     Z=1
    )) {module=ScaleMeshSize002}
   &ScaleMeshSize(Scale Factor=&MakeCustomVectorFromBool(Bool=&RandomBool(), True Vector=<1 -1 -1>, False Vector=<1 1 1>))
   &DynamicMaterialParameters(
    Index 0 Param 1=&FloatFromCurve(
     FloatCurve=DataInterface() {propertyValues {Curve {keys [{interpMode=RCIM_Cubic, tangentMode=RCTM_Break, time=0, value=0.2}, {arriveTangent=2.372, tangentMode=RCTM_User, time=0.8, value=1}]}}}
     Scale Curve=&RandomRangeFloat(Minimum=1, Maximum=1.2)
    )
    Index 0 Param 2=0.7
    Index 0 Param 3=&FloatFromCurve(
     FloatCurve=DataInterface() {propertyValues {Curve {keys [
         {arriveTangent=1.5447, interpMode=RCIM_Cubic, leaveTangent=1.5447, tangentMode=RCTM_User, time=0, value=0}
         {arriveTangent=0.0259, interpMode=RCIM_Cubic, leaveTangent=0.0259, tangentMode=RCTM_User, time=1, value=3}
        ]}}}
     Scale Curve=&RandomRangeFloat(Minimum=1, Maximum=2)
    )
    Index 1 Param 1=&Multiply_Float(
     A=&RandomRangeFloat(Minimum=1, Maximum=2)
     B=&Lerp_Float(A=0.65, B=0.8, Alpha=$User.ForwardDot)
    )
    Index 1 Param 2=&RandomRangeFloat(Minimum=0, Maximum=1)
    Index 1 Param 3=&Lerp_Float(A=0.8, B=0.8, Alpha=$User.ForwardDot)
    Index 1 Param 4=1
    Index 2 Param 1=&FloatFromCurve(FloatCurve=DataInterface() {propertyValues {Curve {keys [{tangentMode=RCTM_User, time=0, value=0.2}, {tangentMode=RCTM_User, time=1, value=1}]}}})
    Index 2 Param 4=5
    Module.Index0Float0Write=true
    Module.Index0Float1Write=true
    Module.Index0Float2Write=true
    Module.Index0Float3Write=true
    Module.Index1Float0Write=true
    Module.Index1Float1Write=true
    Module.Index1Float2Write=true
    Module.Index1Float3Write=true
    Module.Index2Float0Write=true
    Module.Index2Float1Write=true
    Module.Index2Float2Write=true
    Module.Index2Float3Write=true
    Write Parameter Index 1=::NewEnumerator1
    Write Parameter Index 2=::NewEnumerator1
   )
   &FN_Mesh_Attributes(
    Color=&MakeLinearColorFromVectorAndFloat(
     "Vector (RGB)"=&Multiply_VectorByFloat(Vector=&MakeVectorFromLinearColorRGB(Color=$Particles.Color), Float=8)
     "Float (Alpha)"=&FloatFromCurve(
      FloatCurve=DataInterface() {propertyValues {Curve {keys [
          <at 0 0>
          <at 0.1 0>
          <at 0.2 0.8 RCIM_Cubic>
          {arriveTangent=-2.3099, interpMode=RCIM_Cubic, leaveTangent=-2.3099, tangentMode=RCTM_User, time=1, value=0}
         ]}}}
      Scale Curve=6
     )
    )
    Create Color Attribute=true
    Create Dynamic Parameter Attribute=true
    Index 0 Param 1:NiagaraFloat=&FloatFromCurve(FloatCurve=DataInterface() {propertyValues {Curve {keys [{tangentMode=RCTM_User, time=0, value=0.2}, {tangentMode=RCTM_User, time=1, value=1}]}}})
    Index 0 Param 2=0.7
    Index 0 Param 3:NiagaraFloat=&FloatFromCurve(
     FloatCurve=DataInterface() {propertyValues {Curve {keys [<at 0 0 RCIM_Cubic>, {arriveTangent=2.5076, interpMode=RCIM_Cubic, leaveTangent=2.5076, tangentMode=RCTM_User, time=1, value=3}]}}}
     Scale Curve=&RandomRangeFloat(Minimum=2, Maximum=1)
    )
    Module Output=::NewEnumerator1
    Module.Index0Float0Write=true
    Module.Index0Float1Write=true
    Module.Index0Float2Write=true
    Module.Index0Float3Write=true
   )
   &FN_Mesh_Attributes(
    Color=&MakeLinearColorFromVectorAndFloat(
     "Vector (RGB)"=&Multiply_VectorByFloat(
      Vector=&MakeVectorFromLinearColorRGB(Color=$Particles.Color)
      Float=&FN_FrontEndFloatModifier(
       Gameplay=&SelectFloatFromArray(
        Direct Array Index=&MakeIntFromFloat(Float=$User.VariantSwitch)
        Float Selection Array=DataInterface() {propertyValues {FloatData [2,2,1.5]}}
       )
       Front End=&SelectFloatFromArray(
        Direct Array Index=&MakeIntFromFloat(Float=$User.VariantSwitch)
        Float Selection Array=DataInterface() {propertyValues {FloatData [2,1,1]}}
       )
      )
     )
     "Float (Alpha)"=&FloatFromCurve(
      FloatCurve=DataInterface() {propertyValues {Curve {keys [
          <at 0 1>
          {arriveTangent=1.3985, time=0.2, value=1}
          {arriveTangent=-2, interpMode=RCIM_Cubic, leaveTangent=-2, time=0.35, value=1}
          {arriveTangent=-1.3423, interpMode=RCIM_Cubic, leaveTangent=-1.3423, tangentMode=RCTM_User, time=0.7, value=0}
         ]}}}
      Scale Curve=&FN_FrontEndFloatModifier(
       Gameplay=2.5
       Front End=&SelectFloatFromArray(
        Direct Array Index=&MakeIntFromFloat(Float=$User.VariantSwitch)
        Float Selection Array=DataInterface() {propertyValues {FloatData [2.5,2.5,1.5]}}
       )
      )
     )
    )
    Create Color Attribute=true
    Create Dynamic Parameter Attribute=true
    Index 0 Param 1:NiagaraFloat=&FloatFromCurve(FloatCurve=DataInterface() {propertyValues {Curve {keys [{tangentMode=RCTM_User, time=0, value=0.2}, {tangentMode=RCTM_User, time=1, value=1}]}}})
    Index 0 Param 2=0.7
    Module Output=::NewEnumerator1
    Module.Index0Float0Write=true
    Module.Index0Float1Write=true
    Module.Index0Float2Write=true
    Module.Index0Float3Write=true
   ) {module=FN_Mesh_Attributes001}
   &FN_Sprite_Attributes(
    Camera Offset=30
    Color=&MakeLinearColorFromVectorAndFloat(
     "Vector (RGB)"=&MakeVectorFromLinearColorRGB(Color=&SelectLinearColorFromArray(
       Array Sampling Mode=::NewEnumerator1
       Direct Array Index=&MakeIntFromFloat(Float=$User.VariantSwitch)
       Color Selection Array=DataInterface() {propertyValues {ColorData [<rgba 0.6933 0.04 2 1>, <rgba 0.04 0.3667 2 1>, <rgba 2 0.04 0.6933 1>]}}
       Random Seed=?
      ))
     "Float (Alpha)"=&Multiply_Float(
      A=&FloatFromCurve(FloatCurve=DataInterface() {propertyValues {Curve {keys [
           {interpMode=RCIM_Cubic, tangentMode=RCTM_User, time=0, value=0}
           {interpMode=RCIM_Cubic, tangentMode=RCTM_User, time=0.3, value=1}
           {arriveTangent=-1.4286, interpMode=RCIM_Cubic, leaveTangent=-1.4286, time=0.7, value=1}
           {tangentMode=RCTM_User, time=1, value=0}
          ]}}})
      B=&Lerp_Float(A=2, B=3, Alpha=$User.ForwardDot)
     )
    )
    Create Camera Offset Attribute=true
    Create Color Attribute=true
    Create Position Attribute=true
    Create SpriteSize Attribute=true
    Module Output:ENiagara_AttributeSamplingApplyOutput=::NewEnumerator1
    Position:NiagaraPosition=$Particles.Position
    SpriteSize=&Multiply_Vector2D_ByFloat(
     Float=&FloatFromCurve(
      FloatCurve=DataInterface() {propertyValues {Curve {keys [
          {interpMode=RCIM_Cubic, tangentMode=RCTM_User, time=0, value=0}
          {interpMode=RCIM_Cubic, tangentMode=RCTM_User, time=0.2, value=1}
          {tangentMode=RCTM_User, time=1, value=0}
         ]}}}
      Scale Curve=2
     )
     Vector2D=&Vector2DFromFloat(Value=50)
    )
   )
   &ScaleColor(
    Color Value To Scale=$Particles.Color
    Scale Alpha=&Multiply_Float(
     A=5
     B=&FloatFromCurve(FloatCurve=DataInterface() {propertyValues {Curve {keys [<at 0 0>,<at 0.1 0>,<at 0.2 0.8>,<at 1 0>]}}})
    )
    Scale RGB=&VectorFromFloat(Value=&RandomRangeFloat(Minimum=0.5, Maximum=1, Evaluation Type=::NewEnumerator1))
    ScaleA=true
    ScaleRGB=true
   )
  ]
  rendererClasses [&NiagaraMeshRendererProperties, &NiagaraSpriteRendererProperties]
  renderers [
   {rendererClass=&NiagaraMeshRendererProperties, rendererIndex=0}
   {rendererClass=&NiagaraMeshRendererProperties, rendererIndex=1}
   {rendererClass=&NiagaraMeshRendererProperties, rendererIndex=2}
   {rendererClass=&NiagaraSpriteRendererProperties, rendererIndex=3}
  ]
 }
 Tail {
  emitterSpawnScript [
   /SetVariables(Emitter.DepthFade=10) {isSetParameters=true, module=SetVariables_#4}
   /SetVariables(Emitter.Width=50) {isSetParameters=true, module=SetVariables_#5}
  ]
  emitterUpdateScript [
   &EmitterState(
    Delay First Loop Only=true
    Inactive Response=::NewEnumerator1
    Life Cycle Mode=::NewEnumerator1
    Loop Delay=&FN_FrontEndFloatModifier(
     Gameplay=&Lerp_Float(
      A=&RandomRangeFloat(Minimum=0.5, Maximum=3)
      B=&RandomRangeFloat(Minimum=0.2, Maximum=0)
      Alpha=&Lerp_Float(A=$System.Test, B=1, Alpha=$User.ForwardDot)
     )
     Front End=&Lerp_Float(A=$System.Front End Delay, B=0, Alpha=$System.Test)
    )
    Loop Duration=&FN_FrontEndFloatModifier(
     Gameplay=&Lerp_Float(
      A=&RandomRangeFloat(Minimum=0.5, Maximum=1)
      B=&RandomRangeFloat(Minimum=0.5, Maximum=0.2)
      Alpha=&Lerp_Float(A=$System.Test, B=1, Alpha=$User.ForwardDot)
     )
     Front End=&RandomRangeFloat(Minimum=0.5, Maximum=0.2)
    )
    Recalculate Duration Each Loop=true
    UseLoopDelay=true
   )
   &SpawnRate(
    SpawnRate=&FN_FrontEndFloatModifier(
     Gameplay=&Lerp_Float(A=2, B=5, Alpha=&Lerp_Float(A=$System.Test, B=1, Alpha=$User.ForwardDot))
     Front End=3
    )
    Use Spawn Probability=false
   )
  ]
  particleSpawnScript [
   &InitializeParticle(
    Color=&SelectLinearColorFromArray(
     Array Sampling Mode=::NewEnumerator1
     Direct Array Index=&MakeIntFromFloat(Float=$User.VariantSwitch)
     Color Selection Array=DataInterface() {propertyValues {ColorData [<rgba 0.0833 0 0.5 1>, <rgba 0 0 0 0>]}}
     Random Seed=?
    )
    Lifetime=&Lerp_Float(
     A=&RandomRangeFloat(Minimum=0.3, Maximum=1)
     B=&RandomRangeFloat(Minimum=0.3, Maximum=0.6)
     Alpha=$User.ForwardDot
    )
    Mesh Scale Max=<1 1 1>
    Mesh Scale Mode=::NewEnumerator2
    UsePositionOffset=false
    Write Material Random=true
   )
   &CameraOffset(Camera Offset Amount=5)
   &InitialMeshOrientation(
    Mesh Orientation Mode=::NewEnumerator2
    Rotation=&RandomRangeVector(Minimum=<0 0 0>, Maximum=<0 1 0>)
    Use Rotation Vector=true
   ) {module=InitialMeshOrientation001}
  ]
  particleUpdateScript [
   &ParticleState()
   &SkeletalMeshLocation(
    Bone / Socket Index=5
    Bone Sampling Mode=::NewEnumerator4
    MaskSpawnGroup=false
    Orientation Sampling=::NewEnumerator1
    Sampled Position Offset=&FN_FrontEndVector3Modifier(Gameplay=<0 0 0>, Front End=<0 0 15>)
    Skeletal Mesh=$System.Skeletal Mesh
   )
   &self:VariantColor(
    DefaultColor:LinearColor=&ColorFromCurve(ColorCurve=DataInterface() {propertyValues {
       AlphaCurve {keys [<at 0.4428 1>,<at 1 0 RCIM_Cubic>]}
       BlueCurve {keys [<at 0.2021 1 RCIM_Cubic>,<at 0.4091 1>]}
       GreenCurve {keys [<at 0.2021 0 RCIM_Cubic>,<at 0.4091 0>]}
       RedCurve {keys [<at 0.2021 0.5 RCIM_Cubic>,<at 0.4091 0.1667>]}
      }})
    PipColor:LinearColor=&ColorFromCurve(ColorCurve=DataInterface() {propertyValues {
       AlphaCurve {keys [<at 0 1 RCIM_Cubic>,<at 1 0 RCIM_Cubic>]}
       BlueCurve {keys [<at 0.1471 1 RCIM_Cubic>,<at 0.4146 1>]}
       GreenCurve {keys [<at 0.1471 0.6667 RCIM_Cubic>,<at 0.4146 0.1667>]}
       RedCurve {keys [<at 0.1471 0 RCIM_Cubic>,<at 0.4146 0>]}
      }})
    SeedColor:LinearColor=<rgba 1 0 0.3333 1>
    StoneColor:LinearColor=<rgba 0 0 0 0>
    VariantIndex:NiagaraInt32=&MakeIntFromFloat(Float=$User.VariantSwitch)
   ) {module=VariantColor}
   &ScaleMeshSize(Scale Factor=&MakeVector(
     X=1
     Y=&FloatFromCurve(
      FloatCurve=DataInterface() {propertyValues {Curve {keys [
          {arriveTangent=4.4858, interpMode=RCIM_Cubic, leaveTangent=4.4858, tangentMode=RCTM_User, time=0, value=0.2}
          {arriveTangent=0.3309, interpMode=RCIM_Cubic, leaveTangent=0.3309, tangentMode=RCTM_User, time=0.1, value=0.8453}
          {arriveTangent=-0.0088, interpMode=RCIM_Cubic, leaveTangent=-0.0088, tangentMode=RCTM_User, time=1, value=1}
         ]}}}
      Scale Curve=&FN_FrontEndFloatModifier(
       Gameplay=&Lerp_Float(
        A=&RandomRangeFloat(Minimum=1, Maximum=1)
        B=&RandomRangeFloat(Minimum=2, Maximum=2.2)
        Alpha=$User.ForwardDot
       )
       Front End=&RandomRangeFloat(Minimum=2, Maximum=2.2)
      )
     )
     Z=1
    )) {module=ScaleMeshSize002}
   &ScaleMeshSize(Scale Factor=&MakeCustomVectorFromBool(Bool=&RandomBool(), True Vector=<-1 1 -1>, False Vector=<1 1 1>))
   &DynamicMaterialParameters(
    Index 0 Param 1=&FloatFromCurve(
     FloatCurve=DataInterface() {propertyValues {Curve {defaultValue=-0.021, keys [{tangentMode=RCTM_User, time=0, value=0}, {tangentMode=RCTM_User, time=1, value=1}]}}}
     Scale Curve=&RandomRangeFloat(Minimum=1, Maximum=1.2)
    )
    Index 0 Param 2=0.7
    Index 0 Param 3=&FloatFromCurve(
     FloatCurve=DataInterface() {propertyValues {Curve {keys [
         {arriveTangent=1.5447, interpMode=RCIM_Cubic, leaveTangent=1.5447, tangentMode=RCTM_User, time=0, value=0}
         {arriveTangent=0.0259, interpMode=RCIM_Cubic, leaveTangent=0.0259, tangentMode=RCTM_User, time=1, value=3}
        ]}}}
     Scale Curve=&RandomRangeFloat(Minimum=1, Maximum=2)
    )
    Index 0 Param 4=5
    Index 1 Param 1=&Multiply_Float(
     A=&RandomRangeFloat(Minimum=1, Maximum=2)
     B=&Lerp_Float(A=0.65, B=0.8, Alpha=$User.ForwardDot)
    )
    Index 1 Param 2=&RandomRangeFloat(Minimum=0, Maximum=1)
    Index 1 Param 3=1
    Index 1 Param 4=1
    Index 2 Param 1=&FloatFromCurve(FloatCurve=DataInterface() {propertyValues {Curve {keys [{tangentMode=RCTM_User, time=0, value=0.2}, {tangentMode=RCTM_User, time=1, value=1}]}}})
    Index 2 Param 4=5
    Module.Index0Float0Write=true
    Module.Index0Float1Write=true
    Module.Index0Float2Write=true
    Module.Index0Float3Write=true
    Module.Index1Float0Write=true
    Module.Index1Float1Write=true
    Module.Index1Float2Write=true
    Module.Index1Float3Write=true
    Module.Index2Float0Write=true
    Module.Index2Float1Write=true
    Module.Index2Float2Write=true
    Module.Index2Float3Write=true
    Write Parameter Index 1=::NewEnumerator1
    Write Parameter Index 2=::NewEnumerator1
   )
   &FN_Mesh_Attributes(
    Color=&MakeLinearColorFromVectorAndFloat(
     "Vector (RGB)"=&Multiply_VectorByFloat(Vector=&MakeVectorFromLinearColorRGB(Color=$Particles.Color), Float=8)
     "Float (Alpha)"=&FloatFromCurve(
      FloatCurve=DataInterface() {propertyValues {Curve {keys [
          <at 0 0>
          <at 0.1 0>
          <at 0.2 0.8 RCIM_Cubic>
          {arriveTangent=-2.3099, interpMode=RCIM_Cubic, leaveTangent=-2.3099, tangentMode=RCTM_User, time=1, value=0}
         ]}}}
      Scale Curve=6
     )
    )
    Create Color Attribute=true
    Create Dynamic Parameter Attribute=true
    Index 0 Param 1:NiagaraFloat=&FloatFromCurve(FloatCurve=DataInterface() {propertyValues {Curve {keys [{tangentMode=RCTM_User, time=0, value=0.2}, {tangentMode=RCTM_User, time=1, value=1}]}}})
    Index 0 Param 2=0.7
    Index 0 Param 3:NiagaraFloat=&FloatFromCurve(
     FloatCurve=DataInterface() {propertyValues {Curve {keys [<at 0 0 RCIM_Cubic>, {arriveTangent=2.5076, interpMode=RCIM_Cubic, leaveTangent=2.5076, tangentMode=RCTM_User, time=1, value=3}]}}}
     Scale Curve=&RandomRangeFloat(Minimum=2, Maximum=1)
    )
    Index 0 Param 4=5
    Module Output=::NewEnumerator1
    Module.Index0Float0Write=true
    Module.Index0Float1Write=true
    Module.Index0Float2Write=true
    Module.Index0Float3Write=true
   )
   &FN_Mesh_Attributes(
    Color=&MakeLinearColorFromVectorAndFloat(
     "Vector (RGB)"=&Multiply_VectorByFloat(
      Vector=&MakeVectorFromLinearColorRGB(Color=$Particles.Color)
      Float=&FN_FrontEndFloatModifier(
       Gameplay=&SelectFloatFromArray(
        Direct Array Index=&MakeIntFromFloat(Float=$User.VariantSwitch)
        Float Selection Array=DataInterface() {propertyValues {FloatData [2,2,1.5]}}
       )
       Front End=&SelectFloatFromArray(
        Direct Array Index=&MakeIntFromFloat(Float=$User.VariantSwitch)
        Float Selection Array=DataInterface() {propertyValues {FloatData [2,1,1]}}
       )
      )
     )
     "Float (Alpha)"=&FloatFromCurve(
      FloatCurve=DataInterface() {propertyValues {Curve {keys [
          <at 0 1>
          {arriveTangent=1.3985, time=0.2, value=1}
          {arriveTangent=-2, interpMode=RCIM_Cubic, leaveTangent=-2, time=0.35, value=1}
          {arriveTangent=-1.3423, interpMode=RCIM_Cubic, leaveTangent=-1.3423, tangentMode=RCTM_User, time=0.7, value=0}
         ]}}}
      Scale Curve=&FN_FrontEndFloatModifier(
       Gameplay=2
       Front End=&SelectFloatFromArray(
        Direct Array Index=&MakeIntFromFloat(Float=$User.VariantSwitch)
        Float Selection Array=DataInterface() {propertyValues {FloatData [2,2,1.5]}}
       )
      )
     )
    )
    Create Color Attribute=true
    Create Dynamic Parameter Attribute=true
    Index 0 Param 1:NiagaraFloat=&FloatFromCurve(FloatCurve=DataInterface() {propertyValues {Curve {keys [{tangentMode=RCTM_User, time=0, value=0.2}, {tangentMode=RCTM_User, time=1, value=1}]}}})
    Index 0 Param 2=0.7
    Index 0 Param 4=5
    Module Output=::NewEnumerator1
    Module.Index0Float0Write=true
    Module.Index0Float1Write=true
    Module.Index0Float2Write=true
    Module.Index0Float3Write=true
   ) {module=FN_Mesh_Attributes001}
   &FN_Sprite_Attributes(
    Camera Offset=30
    Color=&MakeLinearColorFromVectorAndFloat(
     "Vector (RGB)"=&MakeVectorFromLinearColorRGB(Color=&SelectLinearColorFromArray(
       Array Sampling Mode=::NewEnumerator1
       Direct Array Index=&MakeIntFromFloat(Float=$User.VariantSwitch)
       Color Selection Array=DataInterface() {propertyValues {ColorData [<rgba 0.6933 0.04 2 1>, <rgba 0.04 0.3667 2 1>, <rgba 2 0.04 0.6933 1>]}}
       Random Seed=?
      ))
     "Float (Alpha)"=&Multiply_Float(
      A=&FloatFromCurve(FloatCurve=DataInterface() {propertyValues {Curve {keys [
           {interpMode=RCIM_Cubic, tangentMode=RCTM_User, time=0, value=0}
           {interpMode=RCIM_Cubic, tangentMode=RCTM_User, time=0.3, value=1}
           {arriveTangent=-1.4286, interpMode=RCIM_Cubic, leaveTangent=-1.4286, time=0.7, value=1}
           {tangentMode=RCTM_User, time=1, value=0}
          ]}}})
      B=1
     )
    )
    Create Camera Offset Attribute=true
    Create Color Attribute=true
    Create Position Attribute=true
    Create SpriteSize Attribute=true
    Module Output:ENiagara_AttributeSamplingApplyOutput=::NewEnumerator1
    Position:NiagaraPosition=$Particles.Position
    SpriteSize=&Multiply_Vector2D_ByFloat(
     Float=&FloatFromCurve(
      FloatCurve=DataInterface() {propertyValues {Curve {keys [
          {interpMode=RCIM_Cubic, tangentMode=RCTM_User, time=0, value=0}
          {interpMode=RCIM_Cubic, tangentMode=RCTM_User, time=0.2, value=1}
          {tangentMode=RCTM_User, time=1, value=0}
         ]}}}
      Scale Curve=2
     )
     Vector2D=&Vector2DFromFloat(Value=50)
    )
   )
   &ScaleColor(
    Color Value To Scale=$Particles.Color
    Scale Alpha=&Multiply_Float(
     A=5
     B=&FloatFromCurve(FloatCurve=DataInterface() {propertyValues {Curve {keys [<at 0 0>,<at 0.1 0>,<at 0.2 0.8>,<at 1 0>]}}})
    )
    Scale RGB=&VectorFromFloat(Value=&RandomRangeFloat(Minimum=0.5, Maximum=1, Evaluation Type=::NewEnumerator1))
    ScaleA=true
    ScaleRGB=true
   )
  ]
  rendererClasses [&NiagaraMeshRendererProperties, &NiagaraSpriteRendererProperties]
  renderers [
   {rendererClass=&NiagaraMeshRendererProperties, rendererIndex=0}
   {rendererClass=&NiagaraMeshRendererProperties, rendererIndex=1}
   {rendererClass=&NiagaraMeshRendererProperties, rendererIndex=2}
   {rendererClass=&NiagaraSpriteRendererProperties, rendererIndex=3}
  ]
 }
}
summary {
 systemName=NS_Glider_Shimpilon_Idle
 emitters [
  {emitterName=Tail, rendererClasses [&NiagaraMeshRendererProperties, &NiagaraSpriteRendererProperties]}
  {emitterName=Eyes, rendererClasses [&NiagaraMeshRendererProperties]}
  {emitterName=Forward_smoke, rendererClasses [&NiagaraSpriteRendererProperties]}
  {emitterName=Limbs, rendererClasses [&NiagaraMeshRendererProperties, &NiagaraSpriteRendererProperties]}
 ]
 userVariables [
  User.bIsFrontEnd {type=NiagaraBool}
  User.bIsFrontEndPreview {type=NiagaraBool}
  User.ForwardDot {type=NiagaraFloat, defaultValue=0, description=""}
  User.VariantSwitch {type=NiagaraFloat, defaultValue=0, description=""}
 ]
}
