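# Animation Blueprint for the ShimPilon glider skeleton that drives its kunai bones from the back.
# The input pose runs through four chained parent constraints in component space, one for each left and right kunai_01 and kunai_02 bone.
# Each constraint blends chest, midBack and lowerBack with the same fixed weights, favoring the chest.
# Two self-targeting kunai_01 constraints sit disconnected from the output, and the Event Graph holds no logic.
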AnimBlueprint /BRCosmetics/Gadgets/Assets/VinderTech_GliderChute/Glider_ShimPilon/Meshes/ShimPilon_Glider_AnimBP
target_skeleton=&/ShimPilon_Glider_Skeleton
graphs [
 AnimGraph {nodes [
   n1 = "Constraint - kunai_01_r (P)"!(ComponentPose, Alpha=1, ConstraintWeights_0=1, ConstraintWeights_1:float=0) -> (Pose) {boneToModify {boneName=kunai_01_r}, constraintSetup [{targetBone {boneName=kunai_01_r}}], constraintWeights [1,0]}
   "Constraint - kunai_01_l (P)"!(ComponentPose=@n1, Alpha=1, ConstraintWeights_0=1, ConstraintWeights_1:float=0) -> (Pose) {boneToModify {boneName=kunai_01_l}, constraintSetup [{targetBone {boneName=kunai_01_l}}], constraintWeights [1,0]}
   AnimGraphNode_SubInput_0 = Input Pose()
   n2 = "Constraint - kunai_02_r (PPP)"!(
    ComponentPose=Local To Component(LocalPose=@AnimGraphNode_SubInput_0)
    Alpha=1
    ConstraintWeights_0=0.6123
    ConstraintWeights_1:float=0.3074
    ConstraintWeights_2:float=0.0803
   ) -> (Pose) {
    boneToModify {boneName=kunai_02_r}
    constraintSetup [{targetBone {boneName=chest}}, {targetBone {boneName=midBack}}, {targetBone {boneName=lowerBack}}]
    constraintWeights [1,1,1]
   }
   n3 = "Constraint - kunai_02_l (PPP)"!(ComponentPose=@n2, Alpha=1, ConstraintWeights_0=0.6123, ConstraintWeights_1:float=0.3074, ConstraintWeights_2:float=0.0803) -> (Pose) {
    boneToModify {boneName=kunai_02_l}
    constraintSetup [{targetBone {boneName=chest}}, {targetBone {boneName=midBack}}, {targetBone {boneName=lowerBack}}]
    constraintWeights [1,1,1]
   }
   n4 = "Constraint - kunai_01_r (PPP)"!(ComponentPose=@n3, Alpha=1, ConstraintWeights_0=0.6123, ConstraintWeights_1:float=0.3074, ConstraintWeights_2:float=0.0803) -> (Pose) {
    boneToModify {boneName=kunai_01_r}
    constraintSetup [{targetBone {boneName=chest}}, {targetBone {boneName=midBack}}, {targetBone {boneName=lowerBack}}]
    constraintWeights [1,1,1]
   }
   n5 = "Constraint - kunai_01_l (PPP)"!(ComponentPose=@n4, Alpha=1, ConstraintWeights_0=0.6123, ConstraintWeights_1:float=0.3074, ConstraintWeights_2:float=0.0803) -> (Pose) {
    boneToModify {boneName=kunai_01_l}
    constraintSetup [{targetBone {boneName=chest}}, {targetBone {boneName=midBack}}, {targetBone {boneName=lowerBack}}]
    constraintWeights [1,1,1]
   }
   Output Pose(Result=Component To Local(ComponentPose=@n5)) {layerGroup=None}
  ]}
 EventGraph {nodes [
   TryGetPawnOwner()
   Event BlueprintUpdateAnimation()
  ]}
]
variables []
